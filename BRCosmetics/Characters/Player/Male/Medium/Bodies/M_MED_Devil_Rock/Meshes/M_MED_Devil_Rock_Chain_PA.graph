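# Physics asset for the dangling chain on the right forearm of the medium male Devil Rock player body.
# Five small dyn_arm_Chain capsules link in series with limited swing and twist, anchored through lowerarm_twist_01_r to lowerarm_r.
# Larger capsules cover only the pelvis, right thigh and right forearm, apparently as collision volumes for the chain.
# The thigh and forearm constraints to the pelvis leave swing and twist free.
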PhysicsAsset /BRCosmetics/Characters/Player/Male/Medium/Bodies/M_MED_Devil_Rock/Meshes/M_MED_Devil_Rock_Chain_PA
bodies {
 dyn_arm_Chain_01 [{shapeType=Capsule, length=2, radius=1.5, center=<1 0 0>, rotation=<rot 90 0 0>}]
 dyn_arm_Chain_02 [{shapeType=Capsule, length=2, radius=1.5, center=<1 0 0>, rotation=<rot 90 0 0>}]
 dyn_arm_Chain_03 [{shapeType=Capsule, length=2, radius=1.5, center=<1 0 0>, rotation=<rot 90 0 0>}]
 dyn_arm_Chain_04 [{shapeType=Capsule, length=2, radius=1.5, center=<1 0 0>, rotation=<rot 90 0 0>}]
 dyn_arm_Chain_05 [{shapeType=Capsule, length=3.1834, radius=1.4777, center=<1 0 0>, rotation=<rot 90 0 0>}]
 lowerarm_r [{shapeType=Capsule, length=12.2845, radius=5.3053, center=<-8.1619 -0.7372 -0.6885>, rotation=<rot 90 0 0>}]
 lowerarm_twist_01_r [{shapeType=Capsule, length=1.6245, radius=1.4938, center=<-1.4143 -0.2522 -0.4588>, rotation=<rot 90 0 0>}]
 pelvis [{shapeType=Capsule, length=12.1177, radius=12.6455, center=<6.8924 -0.5681 1.995>}]
 thigh_r [{shapeType=Capsule, length=14.8273, radius=11.9665, center=<6.8582 -0.2761 2.5042>, rotation=<rot 90 0 0>}]
}
constraints [
 {bone1Name=dyn_arm_Chain_02, bone2Name=dyn_arm_Chain_01, swing1LimitDegrees=45, swing2LimitDegrees=45, twistLimitDegrees=15}
 {bone1Name=dyn_arm_Chain_03, bone2Name=dyn_arm_Chain_02, swing1LimitDegrees=45, swing2LimitDegrees=45, twistLimitDegrees=15}
 {bone1Name=dyn_arm_Chain_04, bone2Name=dyn_arm_Chain_03, swing1LimitDegrees=45, swing2LimitDegrees=45, twistLimitDegrees=15}
 {bone1Name=dyn_arm_Chain_01, bone2Name=lowerarm_twist_01_r, swing1LimitDegrees=45, swing2LimitDegrees=80, twistLimitDegrees=10}
 {bone1Name=lowerarm_twist_01_r, bone2Name=lowerarm_r, swing1LimitDegrees=45, swing2LimitDegrees=45, twistLimitDegrees=15}
 {bone1Name=dyn_arm_Chain_05, bone2Name=dyn_arm_Chain_04, swing1LimitDegrees=45, swing2LimitDegrees=45, twistLimitDegrees=15}
 {
  bone1Name=thigh_r
  bone2Name=pelvis
  swing1LimitDegrees=45
  swing1Motion=Free
  swing2LimitDegrees=45
  swing2Motion=Free
  twistLimitDegrees=45
  twistMotion=Free
 }
 {
  bone1Name=lowerarm_r
  bone2Name=pelvis
  swing1LimitDegrees=45
  swing1Motion=Free
  swing2LimitDegrees=45
  swing2Motion=Free
  twistLimitDegrees=45
  twistMotion=Free
 }
]
