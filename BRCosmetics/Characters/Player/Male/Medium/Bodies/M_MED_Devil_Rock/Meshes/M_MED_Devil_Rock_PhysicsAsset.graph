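# Physics asset for the Devil Rock medium male body cosmetic, defining bodies and joints for its dangling accessories.
# Two hip pouches and a clip hang from the pelvis, a four-link chain from the left forearm twist bone.
# Three shell-belt boxes share a pivot on the right thigh.
# Accessory joints have limited swing and twist, while thigh, forearm and pivot links rotate freely, leaving those bodies as anchors or collision shapes.
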PhysicsAsset /BRCosmetics/Characters/Player/Male/Medium/Bodies/M_MED_Devil_Rock/Meshes/M_MED_Devil_Rock_PhysicsAsset
bodies {
 dyn_arm_Chain_01 [{shapeType=Capsule, length=2, radius=1.5, center=<1 0 0>, rotation=<rot 90 0 0>}]
 dyn_arm_Chain_02 [{shapeType=Capsule, length=2, radius=1.5, center=<1 0 0>, rotation=<rot 90 0 0>}]
 dyn_arm_Chain_03 [{shapeType=Capsule, length=2, radius=1.5, center=<1 0 0>, rotation=<rot 90 0 0>}]
 dyn_arm_Chain_04 [{shapeType=Capsule, length=2, radius=1.5, center=<1 0 0>, rotation=<rot 90 0 0>}]
 dyn_clip [{shapeType=Box, extentX=18.2119, extentY=10.9716, extentZ=2.4301, center=<7.6767 0.1908 -0.3572>, rotation=<rot -1.6929 0 0>}]
 dyn_hip_pouch_back [{shapeType=Box, extentX=15.2214, extentY=5.3434, extentZ=14.7393, center=<2.211 -2.5059 0.4713>}]
 dyn_hip_pouch_front [{shapeType=Box, extentX=13.7726, extentY=8.2043, extentZ=4.4929, center=<1.6443 -0.0075 -2.0995>}]
 dyn_shell_belt_back [{shapeType=Box, extentX=8.4213, extentY=1.01, extentZ=8.7156, center=<1.0811 0 0>}]
 dyn_shell_belt_front [{shapeType=Box, extentX=7.0177, extentY=4.9588, extentZ=8.1522, center=<0.4651 0 0>}]
 dyn_shell_belt_pivot [{shapeType=Capsule, length=0.505, radius=0.505, rotation=<rot 0 -90 0>}]
 dyn_shell_belt_side [{shapeType=Box, extentX=9.9167, extentY=11.4398, extentZ=3.581, center=<0.0701 0 0>}]
 lowerarm_twist_01_l [{shapeType=Capsule, length=3.8463, radius=1.7306, center=<1.4435 -0.0066 -0.9117>, rotation=<rot 89.5785 -179.9981 -179.9981>}]
 pelvis [{shapeType=Capsule, length=11.5, radius=9.5303, center=<3.3459 -0.5678 -1.3534>}]
 thigh_l [{shapeType=Capsule, length=16.128, radius=7.9596, center=<-16.292 0.0006 0>, rotation=<rot 90 0 0>}]
 thigh_r [{shapeType=Capsule, length=15.1088, radius=7.7078, center=<15.2408 1.1829 -0.6283>, rotation=<rot 90 0 0>}]
}
constraints [
 {bone1Name=dyn_hip_pouch_front, bone2Name=pelvis, swing1LimitDegrees=45, swing2LimitDegrees=45, twistLimitDegrees=15}
 {bone1Name=dyn_clip, bone2Name=pelvis, swing1LimitDegrees=45, swing2LimitDegrees=45, twistLimitDegrees=8}
 {bone1Name=dyn_hip_pouch_back, bone2Name=pelvis, swing1LimitDegrees=30, swing2LimitDegrees=45, twistLimitDegrees=15}
 {
  bone1Name=thigh_l
  bone2Name=pelvis
  swing1LimitDegrees=45
  swing1Motion=Free
  swing2LimitDegrees=45
  swing2Motion=Free
  twistLimitDegrees=45
  twistMotion=Free
 }
 {
  bone1Name=thigh_r
  bone2Name=pelvis
  swing1LimitDegrees=45
  swing1Motion=Free
  swing2LimitDegrees=45
  swing2Motion=Free
  twistLimitDegrees=45
  twistMotion=Free
 }
 {
  bone1Name=dyn_shell_belt_pivot
  bone2Name=thigh_r
  swing1LimitDegrees=45
  swing1Motion=Free
  swing2LimitDegrees=45
  swing2Motion=Free
  twistLimitDegrees=45
  twistMotion=Free
 }
 {bone1Name=dyn_shell_belt_back, bone2Name=dyn_shell_belt_pivot, swing1LimitDegrees=45, swing2LimitDegrees=45, twistLimitDegrees=15}
 {bone1Name=dyn_shell_belt_side, bone2Name=dyn_shell_belt_pivot, swing1LimitDegrees=45, swing2LimitDegrees=45, twistLimitDegrees=15}
 {bone1Name=dyn_shell_belt_front, bone2Name=dyn_shell_belt_pivot, swing1LimitDegrees=45, swing2LimitDegrees=45, twistLimitDegrees=15}
 {bone1Name=dyn_arm_Chain_01, bone2Name=lowerarm_twist_01_l, swing1LimitDegrees=80, swing2LimitDegrees=80, twistLimitDegrees=15}
 {bone1Name=dyn_arm_Chain_02, bone2Name=dyn_arm_Chain_01, swing1LimitDegrees=45, swing2LimitDegrees=45, twistLimitDegrees=15}
 {bone1Name=dyn_arm_Chain_03, bone2Name=dyn_arm_Chain_02, swing1LimitDegrees=45, swing2LimitDegrees=45, twistLimitDegrees=15}
 {bone1Name=dyn_arm_Chain_04, bone2Name=dyn_arm_Chain_03, swing1LimitDegrees=45, swing2LimitDegrees=45, twistLimitDegrees=15}
 {
  bone1Name=lowerarm_twist_01_l
  bone2Name=pelvis
  swing1LimitDegrees=45
  swing1Motion=Free
  swing2LimitDegrees=45
  swing2Motion=Free
  twistLimitDegrees=45
  twistMotion=Free
 }
]
