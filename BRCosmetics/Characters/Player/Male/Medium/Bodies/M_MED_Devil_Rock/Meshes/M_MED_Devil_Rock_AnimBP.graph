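# Post-pose Anim Blueprint for the Devil Rock medium male cosmetic body, layering secondary motion onto the attached parent mesh's copied pose.
# Anim Dynamics swings the shell belt pivot relative to the pelvis, then two RigidBody nodes simulate the body and chain physics assets.
# The chain pass takes a skydiving force vector as external force.
# A Control Rig receives the crouch state and finishes the pose; the event graph is empty scaffolding.
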AnimBlueprint /BRCosmetics/Characters/Player/Male/Medium/Bodies/M_MED_Devil_Rock/Meshes/M_MED_Devil_Rock_AnimBP
target_skeleton=&/M_MED_Devil_Rock_Skeleton
graphs [
 AnimGraph {nodes [
   n1 = Anim Dynamics - Bone: dyn_shell_belt_pivot(ComponentPose=Local To Component(LocalPose=Copy Pose From Mesh() {bUseAttachedParent=true}), Alpha=1, ExternalForce="0,0,0") {
    angularDampingOverride=0.99
    bGravityOverrideInSimSpace=true
    bOverrideAngularDamping=true
    bUseGravityOverride=true
    simulationSpace=BoneRelative
    alphaBoolBlend {blendOption=Linear}
    boundBone {boneName=dyn_shell_belt_pivot}
    componentAppliedLinearAccClamp=<10000 10000 10000>
    componentLinearAccScale=<0.3 0.3 0.3>
    gravityOverride {z=-980}
    physicsBodyDefinitions [{
      sphereCollisionRadius=0
      boundBone {boneName=dyn_shell_belt_pivot}
      boxExtents=<10 15 30>
      constraintSetup {angularLimitsMax=<3 1 30>, angularLimitsMin=<-3 -45 -30>, angularTarget=<0 0 0>}
      localJointOffset=<0 0 -2>
     }]
    relativeSpaceBone {boneName=pelvis}
   }
   n2 = RigidBody(ComponentPose=@n1) {
    bClampLinearTranslationLimitToRefPose=true
    bOverrideWorldGravity=true
    overridePhysicsAsset=&/M_MED_Devil_Rock_PhysicsAsset
    simulationSpace=BaseBoneSpace
    alphaBoolBlend {blendOption=Linear}
    baseBoneRef {boneName=root}
    componentLinearAccScale=<0.3 0.3 0.3>
    overrideWorldGravity {z=-980}
   }
   n3 = RigidBody(
    ComponentPose=@n2
    ExternalForce=Break Fort Anim Input Skydiving External Force(FortAnimInput_SkydivingExternalForce=Make Fort Anim Input Skydiving External Force!() -> (FortAnimInput_SkydivingExternalForce:Fort Anim Input Skydiving External Force Structure) {is=title K2Node_MakeStruct}) -> (
     bUseSkydivingVectorForce=~
     SkydivingVectorForce=~
     HeadToPelvisDirection=~
     FloatingMultiplier=~
     FloatingAdditive=~
     DivingMultiplier=~
     DivingAdditive=~
     ParachutingMultiplier=~
     ParachutingAdditive=~
    ).SkydivingVectorForce
   ) {
    bClampLinearTranslationLimitToRefPose=true
    bOverrideWorldGravity=true
    overridePhysicsAsset=&/M_MED_Devil_Rock_Chain_PA
    simulationSpace=BaseBoneSpace
    alphaBoolBlend {blendOption=Linear}
    baseBoneRef {boneName=root}
    componentAppliedLinearAccClamp=<5000 5000 5000>
    componentLinearAccScale=<0.05 0.05 0.05>
    overrideWorldGravity {z=-980}
   }
   n4 = Control Rig!(Source=Component To Local(ComponentPose=@n3), Alpha, is_crouching:bool=Get bIsCrouching()) -> (Pose) {
    alphaBoolBlend {blendOption=Linear}
    controlRigAssetReference {blueprintRigClass=&/M_MED_Devil_Rock_CtrlRig.M_MED_Devil_Rock_CtrlRig_C}
   }
   Output Pose(Result=@n4) {layerGroup=None}
  ]}
 EventGraph {nodes [
   TryGetPawnOwner()
   Event BlueprintUpdateAnimation()
  ]}
]
variables []
