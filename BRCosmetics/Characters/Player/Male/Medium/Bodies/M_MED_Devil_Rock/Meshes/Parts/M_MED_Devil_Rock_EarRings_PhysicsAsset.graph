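# Physics asset for the earrings of the male medium Devil Rock player cosmetic, defining physics bodies for four earring bones and the head.
# Small identical spheres sit on the left and right "top" and "brm" earring bones, with a single capsule covering the head.
# Each earring is constrained directly to the head rather than chained to another earring, with matching swing limits and a tighter twist limit.
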PhysicsAsset /BRCosmetics/Characters/Player/Male/Medium/Bodies/M_MED_Devil_Rock/Meshes/Parts/M_MED_Devil_Rock_EarRings_PhysicsAsset
bodies {
 earring_brm_l [{shapeType=Sphere, radius=0.6183, center=<0.1 0 0>}]
 earring_brm_r [{shapeType=Sphere, radius=0.6183, center=<0.1 0 0>}]
 earring_top_l [{shapeType=Sphere, radius=0.6183, center=<0.1 0 0>}]
 earring_top_r [{shapeType=Sphere, radius=0.6183, center=<0.1 0 0>}]
 head [{shapeType=Capsule, length=4.25, radius=6.8353, center=<6.4501 1.0969 -0.1574>, rotation=<rot 0 0 -90>}]
}
constraints [
 {bone1Name=earring_top_l, bone2Name=head, swing1LimitDegrees=45, swing2LimitDegrees=45, twistLimitDegrees=15}
 {bone1Name=earring_brm_l, bone2Name=head, swing1LimitDegrees=45, swing2LimitDegrees=45, twistLimitDegrees=15}
 {bone1Name=earring_top_r, bone2Name=head, swing1LimitDegrees=45, swing2LimitDegrees=45, twistLimitDegrees=15}
 {bone1Name=earring_brm_r, bone2Name=head, swing1LimitDegrees=45, swing2LimitDegrees=45, twistLimitDegrees=15}
]
