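# Physics asset for the face accessory mesh of the Devil Rock medium male player cosmetic, covering its dynamic hair and ponytail bones.
# Capsule bodies cover the clavicles, neck, head and each dynamic bone.
# Side hair forms branching chains from the head, while the ponytail runs from a base through two main segments into branches A–G.
# Strand joints have swing and twist limits; head-to-neck and ponytail base-to-head rotate freely, and clavicles are unconstrained.
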PhysicsAsset /BRCosmetics/Characters/Player/Male/Medium/Bodies/M_MED_Devil_Rock/Meshes/Parts/M_MED_Devil_Rock_FaceAcc_PhysicsAsset
bodies {
 clavicle_l [{shapeType=Capsule, length=4.135, radius=5.3249, center=<5.0135 -0.3458 4.1605>, rotation=<rot 65.1132 0 0>}]
 clavicle_r [{shapeType=Capsule, length=4.135, radius=5.3249, center=<-5.0135 0.3458 -4.1605>, rotation=<rot 65.1132 0 180>}]
 dyn_hair_l_01 [{shapeType=Capsule, length=0.906, radius=0.6821, center=<1 0 0>, rotation=<rot 90 0 0>}]
 dyn_hair_l_02 [{shapeType=Capsule, length=1.6177, radius=0.8498, center=<1 0 0>, rotation=<rot 90 0 0>}]
 dyn_hair_l_03 [{shapeType=Capsule, length=2.8241, radius=1.9259, center=<1 0 0>, rotation=<rot 90 0 0>}]
 dyn_hair_l_04 [{shapeType=Capsule, length=1.5239, radius=1.617, center=<1 0 0>, rotation=<rot 90 0 0>}]
 dyn_hair_l_05 [{shapeType=Capsule, length=7.5137, radius=1.6245, center=<3.354 0 0.4955>, rotation=<rot 84.074 179.9998 179.9998>}]
 dyn_hair_l_06 [{shapeType=Capsule, length=1.1262, radius=0.8617, center=<0.5 0 0>, rotation=<rot 90 0 0>}]
 dyn_hair_l_07 [{shapeType=Capsule, length=2.6578, radius=1.3386, center=<1.8096 0.0001 -0.1936>, rotation=<rot 81.5918 0 0.0001>}]
 dyn_hair_r_01 [{shapeType=Capsule, length=0.9246, radius=0.8999, center=<1 0 0>, rotation=<rot 90 0 0>}]
 dyn_hair_r_010 [{shapeType=Capsule, length=1.9331, radius=1.2912, center=<1 0 0>, rotation=<rot 90 0 0>}]
 dyn_hair_r_011 [{shapeType=Capsule, length=1.4461, radius=1.7526, center=<1 0 0>, rotation=<rot 90 0 0>}]
 dyn_hair_r_012 [{shapeType=Capsule, length=2.3408, radius=1.8653, center=<1 0 0>, rotation=<rot 90 0 0>}]
 dyn_hair_r_013 [{shapeType=Capsule, length=2.9224, radius=0.9939, center=<1 0 0>, rotation=<rot 90 0 0>}]
 dyn_hair_r_02 [{shapeType=Capsule, length=0.505, radius=0.505, center=<1 0 0>, rotation=<rot 90 0 0>}]
 dyn_hair_r_03 [{shapeType=Capsule, length=1.1381, radius=0.7954, center=<1 0 0>, rotation=<rot 90 0 0>}]
 dyn_hair_r_04 [{shapeType=Capsule, length=1.2335, radius=0.8151, center=<1 0 0>, rotation=<rot 90 0 0>}]
 dyn_hair_r_05 [{shapeType=Capsule, length=1.4867, radius=0.5511, center=<1 0 0>, rotation=<rot 90 0 0>}]
 dyn_hair_r_06 [{shapeType=Capsule, length=4.2846, radius=2.5318, center=<1 0 0>, rotation=<rot 90 0 0>}]
 dyn_hair_r_07 [{shapeType=Capsule, length=1.1078, radius=0.9281, center=<1 0 0>, rotation=<rot 90 0 0>}]
 dyn_hair_r_08 [{shapeType=Capsule, length=3.3058, radius=0.7821, center=<1 0 0>, rotation=<rot 90 0 0>}]
 dyn_hair_r_09 [{shapeType=Capsule, length=0.6522, radius=0.6488, center=<1 0 0>}]
 dyn_pnyTail_A_01 [{shapeType=Capsule, length=1.0921, radius=1.2742, center=<1 0 0>, rotation=<rot 90 0 0>}]
 dyn_pnyTail_A_02 [{shapeType=Capsule, length=2.6346, radius=1.1728, center=<1 0 0>, rotation=<rot 90 0 0>}]
 dyn_pnyTail_B_01 [{shapeType=Capsule, length=0.9577, radius=0.72, center=<1 0 0>, rotation=<rot 90 0 0>}]
 dyn_pnyTail_C_01 [{shapeType=Capsule, length=1.1348, radius=0.4941, center=<1 0 0>, rotation=<rot 90 0 0>}]
 dyn_pnyTail_D_01 [{shapeType=Capsule, length=1.3832, radius=1.2922, center=<0.5 0 0>, rotation=<rot 90 0 0>}]
 dyn_pnyTail_D_02 [{shapeType=Capsule, length=2.22, radius=1.1867, center=<0.5 0 0>, rotation=<rot 90 0 0>}]
 dyn_pnyTail_E_01 [{shapeType=Capsule, length=2.5291, radius=1.1311, center=<1 0 0>, rotation=<rot 90 0 0>}]
 dyn_pnyTail_E_02 [{shapeType=Capsule, length=1.4415, radius=0.5955, center=<1 0 0>, rotation=<rot 90 0 0>}]
 dyn_pnyTail_F_01 [{shapeType=Capsule, length=1.4983, radius=1.1353, center=<1 0 0>, rotation=<rot 90 0 0>}]
 dyn_pnyTail_F_02 [{shapeType=Capsule, length=1.1043, radius=0.743, center=<1 0 0>, rotation=<rot 90 0 0>}]
 dyn_pnyTail_G_01 [{shapeType=Capsule, length=2.3341, radius=1.1714, center=<1 0 0>, rotation=<rot 90 0 0>}]
 dyn_pnyTail_base [{shapeType=Capsule, length=0.505, radius=0.505}]
 dyn_pnyTail_main [{shapeType=Capsule, length=1.8316, radius=3.7115, center=<0.5 0 0>, rotation=<rot 90 0 0>}]
 dyn_pnyTail_main_02 [{shapeType=Capsule, length=1.9818, radius=2.2035, center=<0.5 0 0>, rotation=<rot 90 0 0>}]
 head [{shapeType=Capsule, length=4.25, radius=6.8353, center=<6.4501 1.0969 -0.1574>, rotation=<rot 0 0 -90>}]
 neck_01 [{shapeType=Capsule, length=4.2563, radius=5.237, center=<4.3224 0 0>, rotation=<rot 90 0 0>}]
}
constraints [
 {
  bone1Name=dyn_pnyTail_base
  bone2Name=head
  swing1LimitDegrees=45
  swing1Motion=Free
  swing2LimitDegrees=45
  swing2Motion=Free
  twistLimitDegrees=45
  twistMotion=Free
 }
 {bone1Name=dyn_pnyTail_main, bone2Name=dyn_pnyTail_base, swing1LimitDegrees=25, swing2LimitDegrees=25, twistLimitDegrees=15}
 {bone1Name=dyn_pnyTail_main_02, bone2Name=dyn_pnyTail_main, swing1LimitDegrees=20, swing2LimitDegrees=20, twistLimitDegrees=15}
 {bone1Name=dyn_pnyTail_A_01, bone2Name=dyn_pnyTail_main_02, swing1LimitDegrees=15, swing2LimitDegrees=15, twistLimitDegrees=15}
 {bone1Name=dyn_pnyTail_A_02, bone2Name=dyn_pnyTail_A_01, swing1LimitDegrees=15, swing2LimitDegrees=15, twistLimitDegrees=15}
 {bone1Name=dyn_pnyTail_B_01, bone2Name=dyn_pnyTail_main_02, swing1LimitDegrees=15, swing2LimitDegrees=15, twistLimitDegrees=15}
 {bone1Name=dyn_pnyTail_C_01, bone2Name=dyn_pnyTail_main_02, swing1LimitDegrees=15, swing2LimitDegrees=15, twistLimitDegrees=15}
 {bone1Name=dyn_pnyTail_D_01, bone2Name=dyn_pnyTail_main_02, swing1LimitDegrees=20, swing2LimitDegrees=20, twistLimitDegrees=15}
 {bone1Name=dyn_pnyTail_D_02, bone2Name=dyn_pnyTail_D_01, swing1LimitDegrees=15, swing2LimitDegrees=15, twistLimitDegrees=15}
 {bone1Name=dyn_pnyTail_E_01, bone2Name=dyn_pnyTail_main_02, swing1LimitDegrees=15, swing2LimitDegrees=15, twistLimitDegrees=15}
 {bone1Name=dyn_pnyTail_E_02, bone2Name=dyn_pnyTail_E_01, swing1LimitDegrees=15, swing2LimitDegrees=15, twistLimitDegrees=15}
 {bone1Name=dyn_pnyTail_F_01, bone2Name=dyn_pnyTail_main_02, swing1LimitDegrees=15, swing2LimitDegrees=15, twistLimitDegrees=15}
 {bone1Name=dyn_pnyTail_F_02, bone2Name=dyn_pnyTail_F_01, swing1LimitDegrees=15, swing2LimitDegrees=15, twistLimitDegrees=15}
 {bone1Name=dyn_pnyTail_G_01, bone2Name=dyn_pnyTail_main_02, swing1LimitDegrees=15, swing2LimitDegrees=15, twistLimitDegrees=15}
 {bone1Name=dyn_hair_l_01, bone2Name=head, swing1LimitDegrees=45, swing2LimitDegrees=45, twistLimitDegrees=15}
 {bone1Name=dyn_hair_l_02, bone2Name=dyn_hair_l_01, swing1LimitDegrees=45, swing2LimitDegrees=45, twistLimitDegrees=15}
 {bone1Name=dyn_hair_l_03, bone2Name=head, swing1LimitDegrees=30, swing2LimitDegrees=30, twistLimitDegrees=15}
 {bone1Name=dyn_hair_l_04, bone2Name=dyn_hair_l_03, swing1LimitDegrees=30, swing2LimitDegrees=30, twistLimitDegrees=15}
 {bone1Name=dyn_hair_l_05, bone2Name=dyn_hair_l_03, swing1LimitDegrees=45, swing2LimitDegrees=45, twistLimitDegrees=15}
 {bone1Name=dyn_hair_l_06, bone2Name=head, swing1LimitDegrees=30, swing2LimitDegrees=30, twistLimitDegrees=15}
 {bone1Name=dyn_hair_l_07, bone2Name=dyn_hair_l_06, swing1LimitDegrees=45, swing2LimitDegrees=45, twistLimitDegrees=15}
 {bone1Name=dyn_hair_r_01, bone2Name=head, swing1LimitDegrees=30, swing2LimitDegrees=30, twistLimitDegrees=15}
 {bone1Name=dyn_hair_r_02, bone2Name=dyn_hair_r_01, swing1LimitDegrees=30, swing2LimitDegrees=30, twistLimitDegrees=15}
 {bone1Name=dyn_hair_r_03, bone2Name=head, swing1LimitDegrees=30, swing2LimitDegrees=30, twistLimitDegrees=15}
 {bone1Name=dyn_hair_r_04, bone2Name=head, swing1LimitDegrees=15, swing2LimitDegrees=15, twistLimitDegrees=15}
 {bone1Name=dyn_hair_r_05, bone2Name=dyn_hair_r_04, swing1LimitDegrees=15, swing2LimitDegrees=15, twistLimitDegrees=15}
 {bone1Name=dyn_hair_r_06, bone2Name=head, swing1LimitDegrees=45, swing2LimitDegrees=45, twistLimitDegrees=15}
 {bone1Name=dyn_hair_r_07, bone2Name=dyn_hair_r_06, swing1LimitDegrees=30, swing2LimitDegrees=30, twistLimitDegrees=15}
 {bone1Name=dyn_hair_r_08, bone2Name=dyn_hair_r_07, swing1LimitDegrees=30, swing2LimitDegrees=30, twistLimitDegrees=15}
 {bone1Name=dyn_hair_r_09, bone2Name=dyn_hair_r_06, swing1LimitDegrees=45, swing2LimitDegrees=45, twistLimitDegrees=15}
 {bone1Name=dyn_hair_r_010, bone2Name=dyn_hair_r_09, swing1LimitDegrees=45, swing2LimitDegrees=45, twistLimitDegrees=15}
 {bone1Name=dyn_hair_r_011, bone2Name=dyn_hair_r_06, swing1LimitDegrees=30, swing2LimitDegrees=30, twistLimitDegrees=15}
 {bone1Name=dyn_hair_r_012, bone2Name=dyn_hair_r_011, swing1LimitDegrees=15, swing2LimitDegrees=15, twistLimitDegrees=15}
 {bone1Name=dyn_hair_r_013, bone2Name=dyn_hair_r_012, swing1LimitDegrees=15, swing2LimitDegrees=15, twistLimitDegrees=15}
 {
  bone1Name=head
  bone2Name=neck_01
  swing1LimitDegrees=45
  swing1Motion=Free
  swing2LimitDegrees=45
  swing2Motion=Free
  twistLimitDegrees=45
  twistMotion=Free
 }
]
